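# Physics asset for the F_MED_PhoneCharger female player body, setting up capsule bodies and joints for the dynamic shirt bones around hips and upper legs.
# Pelvis and thigh capsules cover the body, with free-rotation joints linking small MSTR root capsules to pelvis or, for front-leg pieces, thighs.
# Shirt segments hang from these roots with narrow swing and twist limits.
# Back and side-back chains have two segments with locked twist; the others have one.
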PhysicsAsset /BRCosmetics/Characters/Player/Female/Medium/Bodies/F_MED_PhoneCharger/Meshes/F_MED_PhoneCharger_Physics
bodies {
 dyn_shirt_bk_l_01 [{shapeType=Capsule, length=13.3631, radius=2.6143, shapeName=dyn_shirt_bk_l_01_capsule, center=<9.8833 2.4811 0>, rotation=<rot 90 0 0.0001>}]
 dyn_shirt_bk_l_02 [{shapeType=Capsule, length=12.6072, radius=3.624, shapeName=dyn_shirt_bk_l_02_capsule, center=<2.5092 2.4811 0>}]
 dyn_shirt_bk_l_MSTR [{shapeType=Capsule, length=1, radius=1, shapeName=dyn_shirt_bk_l_MSTR_capsule}]
 dyn_shirt_bk_r_01 [{shapeType=Capsule, length=12.5775, radius=2.6143, shapeName=dyn_shirt_bk_r_01_capsule, center=<9.8833 2.4811 0>, rotation=<rot 90 0 0.0001>}]
 dyn_shirt_bk_r_02 [{shapeType=Capsule, length=9.3716, radius=3.624, shapeName=dyn_shirt_bk_r_02_capsule, center=<2.5092 2.4811 0>}]
 dyn_shirt_bk_r_MSTR [{shapeType=Capsule, length=1, radius=1, shapeName=dyn_shirt_bk_r_MSTR_capsule}]
 dyn_shirt_fr_01 [{
   shapeType=Capsule
   length=12.1456
   radius=3.2388
   shapeName=dyn_shirt_fr_01_capsule
   center=<10.9736 1.8677 0>
   rotation=<rot 90 -179.9999 179.9999>
  }]
 dyn_shirt_fr_MSTR [{shapeType=Capsule, length=1, radius=1, shapeName=dyn_shirt_fr_MSTR_capsule}]
 dyn_shirt_fr_leg_l_01 [{
   shapeType=Capsule
   length=14.2563
   radius=2.503
   shapeName=dyn_shirt_fr_leg_l_01_capsule
   center=<10.9736 0.3593 0>
   rotation=<rot 90 -0.0006 -0.0007>
  }]
 dyn_shirt_fr_leg_l_MSTR [{shapeType=Capsule, length=1, radius=1, shapeName=dyn_shirt_fr_leg_l_MSTR_capsule}]
 dyn_shirt_fr_leg_r_01 [{
   shapeType=Capsule
   length=13.3879
   radius=2.503
   shapeName=dyn_shirt_fr_leg_r_01_capsule
   center=<10.9736 0.3593 0>
   rotation=<rot 90 0.0001 0>
  }]
 dyn_shirt_fr_leg_r_MSTR [{shapeType=Capsule, length=1, radius=1, shapeName=dyn_shirt_fr_leg_r_MSTR_capsule}]
 dyn_shirt_side_l_01 [{
   shapeType=Capsule
   length=16.3677
   radius=3.3821
   shapeName=dyn_shirt_side_l_01_capsule
   center=<10.9736 1.4001 0>
   rotation=<rot 90 -179.9209 -179.9209>
  }]
 dyn_shirt_side_l_MSTR [{shapeType=Capsule, length=1, radius=1, shapeName=dyn_shirt_side_l_MSTR_capsule}]
 dyn_shirt_side_r_01 [{shapeType=Capsule, length=15.1704, radius=2.0758, shapeName=dyn_shirt_side_r_01_capsule, center=<10.9736 2.15 0>, rotation=<rot 90 180 180>}]
 dyn_shirt_side_r_MSTR [{shapeType=Capsule, length=1, radius=1, shapeName=dyn_shirt_side_r_MSTR_capsule}]
 dyn_shirt_sidebk_l_01 [{
   shapeType=Capsule
   length=10.6558
   radius=2.6143
   shapeName=dyn_shirt_sidebk_l_01_capsule
   center=<9.8833 2.4811 0>
   rotation=<rot 90 -177.4695 -177.4694>
  }]
 dyn_shirt_sidebk_l_02 [{shapeType=Capsule, length=9.8802, radius=3.624, shapeName=dyn_shirt_sidebk_l_02_capsule, center=<2.5092 2.4811 0>}]
 dyn_shirt_sidebk_l_MSTR [{shapeType=Capsule, length=1, radius=1, shapeName=dyn_shirt_sidebk_l_MSTR_capsule}]
 dyn_shirt_sidebk_r_01 [{
   shapeType=Capsule
   length=10.6558
   radius=2.6143
   shapeName=dyn_shirt_sidebk_r_01_capsule
   center=<9.8833 2.4811 0>
   rotation=<rot 90 0.1186 0.1187>
  }]
 dyn_shirt_sidebk_r_02 [{shapeType=Capsule, length=9.5913, radius=3.624, shapeName=dyn_shirt_sidebk_r_02_capsule, center=<2.5092 2.4811 0>}]
 dyn_shirt_sidebk_r_MSTR [{shapeType=Capsule, length=1, radius=1, shapeName=dyn_shirt_sidebk_r_MSTR_capsule}]
 pelvis [{shapeType=Capsule, length=11.8347, radius=11.5803, shapeName=pelvis_capsule, center=<-16.7008 0 0>}]
 thigh_l [{shapeType=Capsule, length=31.6353, radius=9.5598, shapeName=thigh_l_capsule, center=<-23.3024 0 0>, rotation=<rot -90 -179.578 179.5779>}]
 thigh_r [{shapeType=Capsule, length=31.6353, radius=9.5598, shapeName=thigh_r_capsule, center=<22.9778 0.4059 0>, rotation=<rot -90 -179.578 179.5779>}]
}
constraints [
 {
  bone1Name=thigh_l
  bone2Name=pelvis
  swing1LimitDegrees=45
  swing1Motion=Free
  swing2LimitDegrees=45
  swing2Motion=Free
  twistLimitDegrees=45
  twistMotion=Free
 }
 {
  bone1Name=thigh_r
  bone2Name=pelvis
  swing1LimitDegrees=45
  swing1Motion=Free
  swing2LimitDegrees=45
  swing2Motion=Free
  twistLimitDegrees=45
  twistMotion=Free
 }
 {
  bone1Name=dyn_shirt_bk_l_MSTR
  bone2Name=pelvis
  swing1LimitDegrees=45
  swing1Motion=Free
  swing2LimitDegrees=45
  swing2Motion=Free
  twistLimitDegrees=45
  twistMotion=Free
 }
 {
  bone1Name=dyn_shirt_bk_l_01
  bone2Name=dyn_shirt_bk_l_MSTR
  swing1LimitDegrees=45
  swing2LimitDegrees=5
  twistLimitDegrees=1
  twistMotion=Locked
 }
 {
  bone1Name=dyn_shirt_bk_l_02
  bone2Name=dyn_shirt_bk_l_01
  swing1LimitDegrees=45
  swing2LimitDegrees=5
  twistLimitDegrees=1
  twistMotion=Locked
 }
 {
  bone1Name=dyn_shirt_bk_r_MSTR
  bone2Name=pelvis
  swing1LimitDegrees=45
  swing1Motion=Free
  swing2LimitDegrees=45
  swing2Motion=Free
  twistLimitDegrees=45
  twistMotion=Free
 }
 {
  bone1Name=dyn_shirt_bk_r_01
  bone2Name=dyn_shirt_bk_r_MSTR
  swing1LimitDegrees=45
  swing2LimitDegrees=5
  twistLimitDegrees=1
  twistMotion=Locked
 }
 {
  bone1Name=dyn_shirt_bk_r_02
  bone2Name=dyn_shirt_bk_r_01
  swing1LimitDegrees=45
  swing2LimitDegrees=5
  twistLimitDegrees=1
  twistMotion=Locked
 }
 {
  bone1Name=dyn_shirt_fr_MSTR
  bone2Name=pelvis
  swing1LimitDegrees=45
  swing1Motion=Free
  swing2LimitDegrees=45
  swing2Motion=Free
  twistLimitDegrees=45
  twistMotion=Free
 }
 {bone1Name=dyn_shirt_fr_01, bone2Name=dyn_shirt_fr_MSTR, swing1LimitDegrees=30, swing2LimitDegrees=5, twistLimitDegrees=1}
 {
  bone1Name=dyn_shirt_side_l_MSTR
  bone2Name=pelvis
  swing1LimitDegrees=45
  swing1Motion=Free
  swing2LimitDegrees=45
  swing2Motion=Free
  twistLimitDegrees=45
  twistMotion=Free
 }
 {bone1Name=dyn_shirt_side_l_01, bone2Name=dyn_shirt_side_l_MSTR, swing1LimitDegrees=30, swing2LimitDegrees=5, twistLimitDegrees=1}
 {
  bone1Name=dyn_shirt_side_r_MSTR
  bone2Name=pelvis
  swing1LimitDegrees=45
  swing1Motion=Free
  swing2LimitDegrees=45
  swing2Motion=Free
  twistLimitDegrees=45
  twistMotion=Free
 }
 {bone1Name=dyn_shirt_side_r_01, bone2Name=dyn_shirt_side_r_MSTR, swing1LimitDegrees=30, swing2LimitDegrees=5, twistLimitDegrees=1}
 {
  bone1Name=dyn_shirt_sidebk_l_MSTR
  bone2Name=pelvis
  swing1LimitDegrees=45
  swing1Motion=Free
  swing2LimitDegrees=45
  swing2Motion=Free
  twistLimitDegrees=45
  twistMotion=Free
 }
 {
  bone1Name=dyn_shirt_sidebk_l_01
  bone2Name=dyn_shirt_sidebk_l_MSTR
  swing1LimitDegrees=45
  swing2LimitDegrees=5
  twistLimitDegrees=1
  twistMotion=Locked
 }
 {
  bone1Name=dyn_shirt_sidebk_l_02
  bone2Name=dyn_shirt_sidebk_l_01
  swing1LimitDegrees=45
  swing2LimitDegrees=5
  twistLimitDegrees=1
  twistMotion=Locked
 }
 {
  bone1Name=dyn_shirt_sidebk_r_MSTR
  bone2Name=pelvis
  swing1LimitDegrees=45
  swing1Motion=Free
  swing2LimitDegrees=45
  swing2Motion=Free
  twistLimitDegrees=45
  twistMotion=Free
 }
 {
  bone1Name=dyn_shirt_sidebk_r_01
  bone2Name=dyn_shirt_sidebk_r_MSTR
  swing1LimitDegrees=45
  swing2LimitDegrees=5
  twistLimitDegrees=1
  twistMotion=Locked
 }
 {
  bone1Name=dyn_shirt_sidebk_r_02
  bone2Name=dyn_shirt_sidebk_r_01
  swing1LimitDegrees=45
  swing2LimitDegrees=5
  twistLimitDegrees=1
  twistMotion=Locked
 }
 {
  bone1Name=dyn_shirt_fr_leg_l_MSTR
  bone2Name=thigh_l
  swing1LimitDegrees=45
  swing1Motion=Free
  swing2LimitDegrees=45
  swing2Motion=Free
  twistLimitDegrees=45
  twistMotion=Free
 }
 {bone1Name=dyn_shirt_fr_leg_l_01, bone2Name=dyn_shirt_fr_leg_l_MSTR, swing1LimitDegrees=30, swing2LimitDegrees=5, twistLimitDegrees=1}
 {bone1Name=dyn_shirt_fr_leg_r_01, bone2Name=dyn_shirt_fr_leg_r_MSTR, swing1LimitDegrees=30, swing2LimitDegrees=5, twistLimitDegrees=1}
 {
  bone1Name=dyn_shirt_fr_leg_r_MSTR
  bone2Name=thigh_r
  swing1LimitDegrees=45
  swing1Motion=Free
  swing2LimitDegrees=45
  swing2Motion=Free
  twistLimitDegrees=45
  twistMotion=Free
 }
]
