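# Animation Blueprint for the female medium PhoneCharger cosmetic body, adding secondary motion to clothing and accessories.
# The parent's copied pose passes through RigidBody, chained AnimDynamics for sleeves, pants and necklace, fixed bone offsets, then a frontend-aware Control Rig.
# Each update computes the simulations' shared gravity, using default gravity for long backbling.
# It also feeds global and three-region WPO cloth parameters to dynamic materials, enabling wind while skydive floating.
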AnimBlueprint /BRCosmetics/Characters/Player/Female/Medium/Bodies/F_MED_PhoneCharger/Meshes/F_MED_PhoneCharger_AnimBP
target_skeleton=&/F_MED_PhoneCharger_Skeleton
graphs [
 AnimGraph {nodes [
   n1 = Control Rig!(Source=Use cached pose 'Pose3'(), Alpha, isFrontend:bool=Get bIsInFrontEnd(), isNotFrontend:bool=Get NotFrontend() -> (NotFrontend:bool)) -> (Pose) {
    alphaBoolBlend {blendOption=Linear}
    controlRigAssetReference {blueprintRigClass=&/F_MED_PhoneCharger_CtrlRig.F_MED_PhoneCharger_CtrlRig_C}
   }
   Output Pose(Result=@n1) {layerGroup=None}
   n2 = Get Override World Gravity() -> (Override World Gravity:Vector)
   n3 = RigidBody!(
    ComponentPose=Local To Component(LocalPose=Copy Pose From Mesh() {bUseAttachedParent=true})
    Alpha
    OverrideWorldGravity:Vector=@n2
    ExternalForce
   ) -> (Pose) {
    bClampLinearTranslationLimitToRefPose=true
    bOverrideWorldGravity=true
    overridePhysicsAsset=&/F_MED_PhoneCharger_Physics
    simulationSpace=BaseBoneSpace
    alphaBoolBlend {blendOption=Linear}
    baseBoneRef {boneName=spine_01}
    componentAppliedLinearAccClamp=<2000 2000 2000>
    componentLinearAccScale=<0.3 0.3 0.3>
    componentLinearVelScale=<0.3 0.3 0.3>
    overrideWorldGravity {z=-980}
   }
   Save cached pose 'PhysAsset_Pose'(Pose=Component To Local(ComponentPose=@n3)) {cacheName=PhysAsset_Pose}
   n4 = Anim Dynamics - Bone: dyn_sleeve_r_MSTR!(ComponentPose=Local To Component(LocalPose=Use cached pose 'PhysAsset_Pose'()), Alpha=1, GravityOverride:Vector=@n2, ExternalForce="0,0,0") -> (Pose) {
    angularDampingOverride=0.999
    angularSpringConstant=50
    bAngularSpring=true
    bGravityOverrideInSimSpace=true
    bOverrideAngularDamping=true
    bUseGravityOverride=true
    bUseSphericalLimits=true
    simulationSpace=BoneRelative
    alphaBoolBlend {blendOption=Linear}
    boundBone {boneName=dyn_sleeve_r_MSTR}
    componentAppliedLinearAccClamp=<3000 3000 3000>
    componentLinearAccScale=<1 1 1>
    componentLinearVelScale=<1 1 1>
    gravityOverride {z=-980}
    physicsBodyDefinitions [{
      sphereCollisionRadius=0
      boundBone {boneName=dyn_sleeve_r_MSTR}
      boxExtents=<20 20 20>
      constraintSetup {angularLimitsMax=<10 5 5>, angularLimitsMin=<-10 -5 -5>, angularTarget=<0 0 0>}
      localJointOffset=<-2 0 0>
     }]
    relativeSpaceBone {boneName=lowerarm_r}
    sphericalLimits [{limitRadius=7, drivingBone {boneName=upperarm_twist_02_r}, sphereLocalOffset=<2.5338 0 -0.2348>}]
   }
   n5 = Anim Dynamics - Bone: dyn_sleeve_fr_r_MSTR!(ComponentPose=@n4, Alpha=1, GravityOverride:Vector=@n2, ExternalForce="0,0,0") -> (Pose) {
    angularDampingOverride=0.999
    angularSpringConstant=150
    bAngularSpring=true
    bGravityOverrideInSimSpace=true
    bOverrideAngularDamping=true
    bUseGravityOverride=true
    bUseSphericalLimits=true
    simulationSpace=RootRelative
    alphaBoolBlend {blendOption=Linear}
    boundBone {boneName=dyn_sleeve_fr_r_MSTR}
    componentAppliedLinearAccClamp=<3000 3000 3000>
    componentLinearAccScale=<1 1 1>
    componentLinearVelScale=<1 1 1>
    gravityOverride {z=-980}
    physicsBodyDefinitions [{
      sphereCollisionRadius=0
      boundBone {boneName=dyn_sleeve_fr_r_MSTR}
      boxExtents=<20 20 20>
      constraintSetup {angularLimitsMax=<0 8 15>, angularLimitsMin=<0 -15 -5>, angularTarget=<0 0 0>}
      localJointOffset=<4 0 0>
     }]
    relativeSpaceBone {boneName=lowerarm_r}
    sphericalLimits [{limitRadius=6.5, drivingBone {boneName=upperarm_twist_02_r}, sphereLocalOffset=<-4 0 0>}]
   }
   n6 = Anim Dynamics - Bone: dyn_sleeve_out_r_MSTR!(ComponentPose=@n5, Alpha=1, GravityOverride:Vector=@n2, ExternalForce="0,0,0") -> (Pose) {
    angularDampingOverride=0.999
    angularSpringConstant=150
    bAngularSpring=true
    bGravityOverrideInSimSpace=true
    bOverrideAngularDamping=true
    bUseGravityOverride=true
    bUseSphericalLimits=true
    simulationSpace=RootRelative
    alphaBoolBlend {blendOption=Linear}
    boundBone {boneName=dyn_sleeve_out_r_MSTR}
    componentAppliedLinearAccClamp=<3000 3000 3000>
    componentLinearAccScale=<1 1 1>
    componentLinearVelScale=<1 1 1>
    gravityOverride {z=-980}
    physicsBodyDefinitions [{
      sphereCollisionRadius=0
      boundBone {boneName=dyn_sleeve_out_r_MSTR}
      boxExtents=<20 20 20>
      constraintSetup {angularLimitsMax=<0 8 15>, angularLimitsMin=<0 -15 -5>, angularTarget=<0 0 0>}
      localJointOffset=<4 0 0>
     }]
    relativeSpaceBone {boneName=lowerarm_r}
    sphericalLimits [{limitRadius=6.5, drivingBone {boneName=upperarm_twist_02_r}, sphereLocalOffset=<-4 0 0>}]
   }
   n7 = Anim Dynamics - Bone: dyn_sleeve_bk_r_MSTR!(ComponentPose=@n6, Alpha=1, GravityOverride:Vector=@n2, ExternalForce="0,0,0") -> (Pose) {
    angularDampingOverride=0.999
    angularSpringConstant=150
    bAngularSpring=true
    bGravityOverrideInSimSpace=true
    bOverrideAngularDamping=true
    bUseGravityOverride=true
    bUseSphericalLimits=true
    simulationSpace=RootRelative
    alphaBoolBlend {blendOption=Linear}
    boundBone {boneName=dyn_sleeve_bk_r_MSTR}
    componentAppliedLinearAccClamp=<3000 3000 3000>
    componentLinearAccScale=<1 1 1>
    componentLinearVelScale=<1 1 1>
    gravityOverride {z=-980}
    physicsBodyDefinitions [{
      sphereCollisionRadius=0
      boundBone {boneName=dyn_sleeve_bk_r_MSTR}
      boxExtents=<20 20 20>
      constraintSetup {angularLimitsMax=<0 8 15>, angularLimitsMin=<0 -15 -5>, angularTarget=<0 0 0>}
      localJointOffset=<4 0 0>
     }]
    relativeSpaceBone {boneName=lowerarm_r}
    sphericalLimits [{limitRadius=6.5, drivingBone {boneName=upperarm_twist_02_r}, sphereLocalOffset=<-4 0 0>}]
   }
   n8 = Anim Dynamics - Bone: dyn_sleeve_inn_l_MSTR!(ComponentPose=@n7, Alpha=1, GravityOverride:Vector=@n2, ExternalForce="0,0,0") -> (Pose) {
    angularDampingOverride=0.999
    angularSpringConstant=150
    bAngularSpring=true
    bGravityOverrideInSimSpace=true
    bOverrideAngularDamping=true
    bUseGravityOverride=true
    bUseSphericalLimits=true
    simulationSpace=RootRelative
    alphaBoolBlend {blendOption=Linear}
    boundBone {boneName=dyn_sleeve_inn_l_MSTR}
    componentAppliedLinearAccClamp=<3000 3000 3000>
    componentLinearAccScale=<1 1 1>
    componentLinearVelScale=<1 1 1>
    gravityOverride {z=-980}
    physicsBodyDefinitions [{
      sphereCollisionRadius=0
      boundBone {boneName=dyn_sleeve_inn_l_MSTR}
      boxExtents=<20 20 20>
      constraintSetup {angularLimitsMax=<0 8 15>, angularLimitsMin=<0 -15 -5>, angularTarget=<0 0 0>}
      localJointOffset=<4 0 0>
     }]
    relativeSpaceBone {boneName=lowerarm_r}
    sphericalLimits [{limitRadius=6.5, drivingBone {boneName=upperarm_twist_02_r}, sphereLocalOffset=<-4 0 0>}]
   }
   n9 = Anim Dynamics - Bone: dyn_sleeve_l_MSTR!(ComponentPose=@n8, Alpha=1, GravityOverride:Vector=@n2, ExternalForce="0,0,0") -> (Pose) {
    angularDampingOverride=0.999
    angularSpringConstant=50
    bAngularSpring=true
    bGravityOverrideInSimSpace=true
    bOverrideAngularDamping=true
    bUseGravityOverride=true
    bUseSphericalLimits=true
    simulationSpace=BoneRelative
    alphaBoolBlend {blendOption=Linear}
    boundBone {boneName=dyn_sleeve_l_MSTR}
    componentAppliedLinearAccClamp=<3000 3000 3000>
    componentLinearAccScale=<1 1 1>
    componentLinearVelScale=<1 1 1>
    gravityOverride {z=-980}
    physicsBodyDefinitions [{
      sphereCollisionRadius=0
      boundBone {boneName=dyn_sleeve_l_MSTR}
      boxExtents=<20 20 20>
      constraintSetup {angularLimitsMax=<10 5 5>, angularLimitsMin=<-10 -5 -5>, angularTarget=<0 0 0>}
      localJointOffset=<-2 0 0>
     }]
    relativeSpaceBone {boneName=lowerarm_r}
    sphericalLimits [{limitRadius=6.4581, drivingBone {boneName=upperarm_twist_02_l}, sphereLocalOffset=<1.6784 0 0>}]
   }
   n10 = Anim Dynamics - Bone: dyn_sleeve_fr_l_MSTR!(ComponentPose=@n9, Alpha=1, GravityOverride:Vector=@n2, ExternalForce="0,0,0") -> (Pose) {
    angularDampingOverride=0.999
    angularSpringConstant=150
    bAngularSpring=true
    bGravityOverrideInSimSpace=true
    bOverrideAngularDamping=true
    bUseGravityOverride=true
    bUseSphericalLimits=true
    simulationSpace=RootRelative
    alphaBoolBlend {blendOption=Linear}
    boundBone {boneName=dyn_sleeve_fr_l_MSTR}
    componentAppliedLinearAccClamp=<3000 3000 3000>
    componentLinearAccScale=<1 1 1>
    componentLinearVelScale=<1 1 1>
    gravityOverride {z=-980}
    physicsBodyDefinitions [{
      sphereCollisionRadius=0
      boundBone {boneName=dyn_sleeve_fr_l_MSTR}
      boxExtents=<20 20 20>
      constraintSetup {angularLimitsMax=<0 15 15>, angularLimitsMin=<0 -15 -10>, angularTarget=<0 0 0>}
      localJointOffset=<2 0 0>
     }]
    relativeSpaceBone {boneName=lowerarm_r}
    sphericalLimits [{limitRadius=8, drivingBone {boneName=upperarm_twist_02_l}, sphereLocalOffset=<5.8412 0 0>}]
   }
   n11 = Anim Dynamics - Bone: dyn_sleeve_out_l_MSTR!(ComponentPose=@n10, Alpha=1, GravityOverride:Vector=@n2, ExternalForce="0,0,0") -> (Pose) {
    angularDampingOverride=0.999
    angularSpringConstant=150
    bAngularSpring=true
    bGravityOverrideInSimSpace=true
    bOverrideAngularDamping=true
    bUseGravityOverride=true
    bUseSphericalLimits=true
    simulationSpace=RootRelative
    alphaBoolBlend {blendOption=Linear}
    boundBone {boneName=dyn_sleeve_out_l_MSTR}
    componentAppliedLinearAccClamp=<3000 3000 3000>
    componentLinearAccScale=<1 1 1>
    componentLinearVelScale=<1 1 1>
    gravityOverride {z=-980}
    physicsBodyDefinitions [{
      sphereCollisionRadius=0
      boundBone {boneName=dyn_sleeve_out_l_MSTR}
      boxExtents=<20 20 20>
      constraintSetup {angularLimitsMax=<0 15 15>, angularLimitsMin=<0 -15 -15>, angularTarget=<0 0 0>}
      localJointOffset=<2 0 0>
     }]
    relativeSpaceBone {boneName=lowerarm_r}
    sphericalLimits [{limitRadius=8, drivingBone {boneName=upperarm_twist_02_l}, sphereLocalOffset=<4 0 0>}]
   }
   n12 = Anim Dynamics - Bone: dyn_sleeve_bk_l_MSTR!(ComponentPose=@n11, Alpha=1, GravityOverride:Vector=@n2, ExternalForce="0,0,0") -> (Pose) {
    angularDampingOverride=0.999
    angularSpringConstant=150
    bAngularSpring=true
    bGravityOverrideInSimSpace=true
    bOverrideAngularDamping=true
    bUseGravityOverride=true
    bUseSphericalLimits=true
    simulationSpace=RootRelative
    alphaBoolBlend {blendOption=Linear}
    boundBone {boneName=dyn_sleeve_bk_l_MSTR}
    componentAppliedLinearAccClamp=<3000 3000 3000>
    componentLinearAccScale=<1 1 1>
    componentLinearVelScale=<1 1 1>
    gravityOverride {z=-980}
    physicsBodyDefinitions [{
      sphereCollisionRadius=0
      boundBone {boneName=dyn_sleeve_bk_l_MSTR}
      boxExtents=<20 20 20>
      constraintSetup {angularLimitsMax=<0 10 15>, angularLimitsMin=<0 -15 0>, angularTarget=<0 0 0>}
      localJointOffset=<2 0 0>
     }]
    relativeSpaceBone {boneName=lowerarm_r}
    sphericalLimits [{limitRadius=8, drivingBone {boneName=upperarm_twist_02_l}, sphereLocalOffset=<-0.3608 0.9678 0.0978>}]
   }
   n13 = Anim Dynamics - Bone: dyn_sleeve_inn_l_MSTR(ComponentPose=@n12, Alpha=1, GravityOverride=@n2, ExternalForce="0,0,0") {
    angularDampingOverride=0.999
    angularSpringConstant=150
    bAngularSpring=true
    bGravityOverrideInSimSpace=true
    bOverrideAngularDamping=true
    bUseGravityOverride=true
    bUseSphericalLimits=true
    simulationSpace=RootRelative
    alphaBoolBlend {blendOption=Linear}
    boundBone {boneName=dyn_sleeve_inn_l_MSTR}
    componentAppliedLinearAccClamp=<3000 3000 3000>
    componentLinearAccScale=<1 1 1>
    componentLinearVelScale=<1 1 1>
    gravityOverride {z=-980}
    physicsBodyDefinitions [{
      sphereCollisionRadius=0
      boundBone {boneName=dyn_sleeve_inn_l_MSTR}
      boxExtents=<20 20 20>
      constraintSetup {angularLimitsMax=<0 15 15>, angularLimitsMin=<0 -15 -15>, angularTarget=<0 0 0>}
      localJointOffset=<2 0 0>
     }]
    relativeSpaceBone {boneName=lowerarm_r}
    sphericalLimits [{limitRadius=8, drivingBone {boneName=upperarm_twist_02_l}, sphereLocalOffset=<1.5599 0 0>}]
   }
   Save cached pose 'Pose2'(Pose=Component To Local(ComponentPose=@n13)) {cacheName=Pose2}
   n14 = Anim Dynamics - Bone: dyn_pant_01_r!(ComponentPose=Local To Component(LocalPose=Use cached pose 'Pose2'()), Alpha=1, GravityOverride:Vector=@n2, ExternalForce="0,0,0") -> (Pose) {
    angularDampingOverride=0.8
    angularSpringConstant=50
    bAngularSpring=true
    bOverrideAngularDamping=true
    bUseGravityOverride=true
    simulationSpace=RootRelative
    alphaBoolBlend {blendOption=Linear}
    boundBone {boneName=dyn_pant_01_r}
    componentAppliedLinearAccClamp=<1000 1000 1000>
    componentLinearAccScale=<1 1 1>
    componentLinearVelScale=<1 1 1>
    gravityOverride {z=-980}
    physicsBodyDefinitions [{
      boundBone {boneName=dyn_pant_01_r}
      boxExtents=<15 15 15>
      constraintSetup {angularLimitsMax=<5 8 8>, angularLimitsMin=<-5 -8 -5>, angularTarget=<0 0 0>}
      localJointOffset=<2 0 0>
     }]
    relativeSpaceBone {boneName=thigh_r}
   }
   n15 = Anim Dynamics - Bone: dyn_pant_01_l!(ComponentPose=@n14, Alpha=1, GravityOverride:Vector=@n2, ExternalForce="0,0,0") -> (Pose) {
    angularDampingOverride=0.8
    angularSpringConstant=50
    bAngularSpring=true
    bOverrideAngularDamping=true
    bUseGravityOverride=true
    simulationSpace=RootRelative
    alphaBoolBlend {blendOption=Linear}
    boundBone {boneName=dyn_pant_01_l}
    componentAppliedLinearAccClamp=<1000 1000 1000>
    componentLinearAccScale=<1 1 1>
    componentLinearVelScale=<1 1 1>
    gravityOverride {z=-980}
    physicsBodyDefinitions [{
      boundBone {boneName=dyn_pant_01_l}
      boxExtents=<15 15 15>
      constraintSetup {angularLimitsMax=<15 8 8>, angularLimitsMin=<-15 -8 -8>, angularTarget=<0 0 0>}
      localJointOffset=<2 0 0>
     }]
    relativeSpaceBone {boneName=thigh_l}
   }
   n16 = "Anim Dynamics - Chain: dyn_necklace_01 -> dyn_necklace_02"!(ComponentPose=@n15, Alpha=1, GravityOverride:Vector=@n2, ExternalForce="0,0,0") -> (Pose) {
    angularDampingOverride=0.999
    angularSpringConstant=50
    bAngularSpring=true
    bChain=true
    bOverrideAngularDamping=true
    bUseGravityOverride=true
    bUseSphericalLimits=true
    simulationSpace=RootRelative
    alphaBoolBlend {blendOption=Linear}
    boundBone {boneName=dyn_necklace_01}
    chainEnd {boneName=dyn_necklace_02}
    componentAppliedLinearAccClamp=<2000 2000 2000>
    componentLinearAccScale=<0.2 0.2 0.2>
    componentLinearVelScale=<0.2 0.2 0.2>
    gravityOverride {z=-980}
    physicsBodyDefinitions [
     {
      boundBone {boneName=dyn_necklace_01}
      boxExtents=<15 15 15>
      constraintSetup {angularLimitsMax=<0 10 30>, angularLimitsMin=<0 -10 0>, angularTarget=<0 0 0>}
      localJointOffset=<-1 0 0>
     }
     {
      boundBone {boneName=dyn_necklace_02}
      boxExtents=<15 15 15>
      constraintSetup {angularLimitsMax=<0 10 30>, angularLimitsMin=<0 -10 0>, angularTarget=<0 0 0>}
      localJointOffset=<-1 0 0>
     }
    ]
    sphericalLimits [{limitRadius=15, drivingBone {boneName=spine_05}, sphereLocalOffset=<-4.1954 -3.9741 0>}]
   }
   n17 = "Transform (Modify) Bone - Bone: dyn_sneaker_r"!(ComponentPose=@n16, bAlphaBoolEnabled:bool=True, Translation="0,-6,0", Rotation="0,0,0", Scale="1,1,1") -> (Pose) {
    alphaInputType=Bool
    rotationMode=BMM_Additive
    translationMode=BMM_Additive
    alphaBoolBlend {blendOption=CubicInOut}
    boneToModify {boneName=dyn_sneaker_r}
    translation {y=-6}
   }
   n18 = "Transform (Modify) Bone - Bone: dyn_sneaker_l"!(ComponentPose=@n17, bAlphaBoolEnabled:bool=True, Translation="0,-6,0", Rotation="0,0,0", Scale="1,1,1") -> (Pose) {
    alphaInputType=Bool
    rotationMode=BMM_Additive
    translationMode=BMM_Additive
    alphaBoolBlend {blendOption=CubicInOut}
    boneToModify {boneName=dyn_sneaker_l}
    translation {y=-6}
   }
   n19 = "Transform (Modify) Bone - Bone: dyn_shirt_sidebk_r_MSTR"!(ComponentPose=@n18, bAlphaBoolEnabled:bool=True, Translation="0,0,0", Rotation="-15,15,0", Scale="1,1,1") -> (Pose) {
    alphaInputType=Bool
    rotationMode=BMM_Additive
    translationMode=BMM_Additive
    alphaBoolBlend {blendOption=Cubic}
    boneToModify {boneName=dyn_shirt_sidebk_r_MSTR}
    rotation {pitch=-15, yaw=15}
   }
   Save cached pose 'Pose3'(Pose=Component To Local(ComponentPose=@n19)) {cacheName=Pose3}
  ]}
 EventGraph {nodes [
   Event OnCharacterCustomizationLoaded()
   n1 = GetPartSkeletalMeshComponent()
   > n2 = Get Dynamic Material Instance(SkeletalMeshComponent=@n1)
   > Set Mat Instance Dynamic(Mat Instance Dynamic=@n2)
   > n3 = GetOrCreateDynamicOverlayMaterial(MeshComponent=@n1, MaterialName=None)
   > Set Overlay Mat Dynamic(Overlay Mat Dynamic:Material Instance Dynamic Object Reference=@n3) -> (Output_Get:Material Instance Dynamic Object Reference)
   n4 = Event BlueprintUpdateAnimation()
   > Set Delta Time X(Delta Time X=@n4.DeltaTimeX)
   > n5 = Sequence()
   > n6 = Gravity Override State Function DEPRECATED(
    CharacterPartAnimInstance=Self-Reference()
    GameTimeInSeconds=Get Game Time in Seconds() -> (Game Time in Seconds:double)
    GravityOverrideParams=Get Gravity Override Params()
   )
   > Set Override World Gravity(Override World Gravity:Vector=SelectVector(A="0,0,-980", B=@n6, bPickA=CheckForBackbling().bIsBackblingLong)) -> (Output_Get:Vector)
   > Set NotFrontend(NotFrontend:bool=NOT Boolean(A=Get bIsInFrontEnd())) -> (Output_Get:bool)
   Is Valid(exec=@n5.then_1, InputObject=Get PlayerPawn() -> (PlayerPawn:Pawn Object Reference))
   > Set WPO Globals Function(
    Material=Get Mat Instance Dynamic()
    Overlay Material=Get Overlay Mat Dynamic() -> (Overlay Mat Dynamic:Material Instance Dynamic Object Reference)
    Skeletal Mesh=Get SkelMesh()
    "Local Noise Transform  Location"="Get Local Noise Transform  Location"() -> ("Local Noise Transform  Location":Vector)
    Local Noise Offset=Get Local Noise Offset()
   )
   > Set WPO Regional Function!(
    execute
    Material:Material Instance Dynamic Object Reference=Get Mat Instance Dynamic()
    Overlay Material:Material Instance Dynamic Object Reference=Get Overlay Mat Dynamic()
    Skeletal Mesh:Skeletal Mesh Component Object Reference=Get SkelMesh()
    Velocity Transform:Name=pelvis
    Velocity Interp Speed:double=Get Velocity Interp Speed()
    Collision Transform:Name=spine_03
    Collision Transform Axis:EAxis Enum
    Invert Collision Transform Axis:bool=false
    Delta Time X:double=Get Delta Time X()
    Previous Velocity Transform Location:"Vector (by ref)"=Get R1_Previous Velocity Transform Location() -> (R1_Previous Velocity Transform Location:Vector)
    Previous Velocity Transform Velocity:"Vector (by ref)"=Get R1_Previous Velocity Transform Velocity() -> (R1_Previous Velocity Transform Velocity:Vector)
    Region:int=1
   ) -> (then)
   > Set WPO Regional Function(
    Material=Get Mat Instance Dynamic()
    Overlay Material=Get Overlay Mat Dynamic()
    Skeletal Mesh=Get SkelMesh()
    Velocity Transform=lowearm_r
    Velocity Interp Speed=Get Velocity Interp Speed()
    Collision Transform=upperarm_r
    Invert Collision Transform Axis=false
    Delta Time X=Get Delta Time X()
    Previous Velocity Transform Location=Get R2_Previous Velocity Transform Location() -> (R2_Previous Velocity Transform Location:Vector)
    Previous Velocity Transform Velocity=Get R2_Previous Velocity Transform Velocity() -> (R2_Previous Velocity Transform Velocity:Vector)
    Region=2
   )
   > Set WPO Regional Function(
    Material=Get Mat Instance Dynamic()
    Overlay Material=Get Overlay Mat Dynamic()
    Skeletal Mesh=Get SkelMesh()
    Velocity Transform=calf_r
    Velocity Interp Speed=Get Velocity Interp Speed()
    Collision Transform=pelvis
    Invert Collision Transform Axis=false
    Delta Time X=Get Delta Time X()
    Previous Velocity Transform Location=Get R3_Previous Velocity Transform Location() -> (R3_Previous Velocity Transform Location:Vector)
    Previous Velocity Transform Velocity=Get R3_Previous Velocity Transform Velocity() -> (R3_Previous Velocity Transform Velocity:Vector)
    Region=3
   )
   > SetScalarParameterValue(self=Get Mat Instance Dynamic(), ParameterName=ClothSim_WindStrength, Value=SelectFloat(A=1, B=0, bPickA=Get bIsSkydiveFloating()))
   Event BlueprintInitializeAnimation()
   > Set SkelMesh(SkelMesh=GetOwnerSkeletalMeshComponent())
   n7 = TryGetPawnOwner()
   > Is Valid(InputObject=@n7)
   > Set PlayerPawn(PlayerPawn:Pawn Object Reference=@n7) -> (Output_Get:Pawn Object Reference)
  ]}
]
variables [
 AlphaRotationHistory {type=number}
 AnimDynamicSleevesAlpha {type=number}
 AnimDynamicsShortsAlpha {type=number}
 Delta Time X {type=number}
 Game Time in Seconds {type=number}
 Gravity Override Params {type=GravityOverrideParamsStruct}
 History {type=PositionHistory, default {range=0, positions []}}
 Local Noise Offset {type=Vector}
 "Local Noise Transform  Location" {type=Vector}
 Mat Instance Dynamic {type=MaterialInstanceDynamic}
 NotFrontend {type=boolean}
 Overlay Mat Dynamic {type=MaterialInstanceDynamic}
 Override World Gravity {type=Vector}
 PlayerPawn {type=Pawn, default=null}
 R1_Previous Velocity Transform Location {type=Vector}
 R1_Previous Velocity Transform Velocity {type=Vector}
 R2_Previous Velocity Transform Location {type=Vector}
 R2_Previous Velocity Transform Velocity {type=Vector}
 R3_Previous Velocity Transform Location {type=Vector}
 R3_Previous Velocity Transform Velocity {type=Vector}
 SkelMesh {type=SkeletalMeshComponent}
 Spine05_Rotation {type=Vector}
 Velocity Interp Speed {type=number, default=1}
]
