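# Physics asset for the hair-bun face accessory of the PhoneCharger female medium character, simulating the two side hair rolls.
# Each side has a small sphere body on its MSTR roll bone and a large capsule on its dyn roll bone.
# A constraint links each dyn bone to its MSTR bone with tight swing and twist limits, keeping bun motion subtle.
# A head capsule provides a body with no constraints attached.
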PhysicsAsset /BRCosmetics/Characters/Player/Female/Medium/Bodies/F_MED_PhoneCharger/Meshes/Parts/F_MED_PhoneCharger_FaceAcc_Buns_Physics
bodies {
 MSTR_hair_left_roll_01 [{shapeType=Sphere, radius=0.505, shapeName=MSTR_hair_left_roll_01_sphere}]
 MSTR_hair_right_roll_01 [{shapeType=Sphere, radius=0.505, shapeName=MSTR_hair_right_roll_01_sphere}]
 dyn_hair_left_roll_01 [{shapeType=Capsule, length=0.75, radius=12, shapeName=hairBun_rt_00_capsule, center=<12 0 0>, rotation=<rot 90 0 0>}]
 dyn_hair_right_roll_01 [{shapeType=Capsule, length=0.75, radius=12, shapeName=hairBun_rt_00_capsule, center=<12 0 0>, rotation=<rot 90 0 0>}]
 head [{shapeType=Capsule, length=1.0928, radius=7.3101, shapeName=head_capsule, center=<5.3553 1.6156 0.0156>, rotation=<rot 90 0 -12.0083>}]
}
constraints [
 {bone1Name=dyn_hair_right_roll_01, bone2Name=MSTR_hair_right_roll_01, swing1LimitDegrees=7, swing2LimitDegrees=7, twistLimitDegrees=1}
 {bone1Name=dyn_hair_left_roll_01, bone2Name=MSTR_hair_left_roll_01, swing1LimitDegrees=7, swing2LimitDegrees=7, twistLimitDegrees=1}
]
