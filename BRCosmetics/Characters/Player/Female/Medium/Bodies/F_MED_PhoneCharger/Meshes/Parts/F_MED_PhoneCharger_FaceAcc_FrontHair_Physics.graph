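# Physics asset for the front hair face-accessory part of the F_MED_PhoneCharger female player body, simulating a single dangling hair strand.
# Four equal-sized capsules on dyn_Hair_01 to 04 form a constraint chain from a small sphere on MSTR_Hair_01.
# Swing limits are tight at the root and loosen toward the tip, with twist limits at zero throughout.
# A large unconstrained head capsule provides a collision volume for the head.
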PhysicsAsset /BRCosmetics/Characters/Player/Female/Medium/Bodies/F_MED_PhoneCharger/Meshes/Parts/F_MED_PhoneCharger_FaceAcc_FrontHair_Physics
bodies {
 MSTR_Hair_01 [{shapeType=Sphere, radius=0.505, shapeName=MSTR_hair_left_roll_01_sphere}]
 dyn_Hair_01 [{shapeType=Capsule, length=2.7796, radius=0.7681, shapeName=dyn_Hair_01_capsule, center=<1.731 0 0>, rotation=<rot 90 180 -180>}]
 dyn_Hair_02 [{shapeType=Capsule, length=2.7796, radius=0.7681, shapeName=dyn_Hair_02_capsule, center=<1.8048 0 0>, rotation=<rot 90 0 0>}]
 dyn_Hair_03 [{shapeType=Capsule, length=2.7796, radius=0.7681, shapeName=dyn_Hair_03_capsule, center=<1.546 0 0>, rotation=<rot 90 0 0>}]
 dyn_Hair_04 [{shapeType=Capsule, length=2.7796, radius=0.7681, shapeName=dyn_Hair_04_capsule, center=<1.8699 0 0>, rotation=<rot 90 180 180>}]
 head [{shapeType=Capsule, length=15.2852, radius=11.351, center=<4.1803 -1.4859 0.0047>, rotation=<rot -0.0474 -68.4162 0.0359>}]
}
constraints [
 {bone1Name=dyn_Hair_04, bone2Name=dyn_Hair_03, swing1LimitDegrees=12, swing2LimitDegrees=30, twistLimitDegrees=0}
 {bone1Name=dyn_Hair_03, bone2Name=dyn_Hair_02, swing1LimitDegrees=12, swing2LimitDegrees=30, twistLimitDegrees=0}
 {bone1Name=dyn_Hair_02, bone2Name=dyn_Hair_01, swing1LimitDegrees=6, swing2LimitDegrees=20, twistLimitDegrees=0}
 {bone1Name=dyn_Hair_01, bone2Name=MSTR_Hair_01, swing1LimitDegrees=3, swing2LimitDegrees=6, twistLimitDegrees=0}
]
