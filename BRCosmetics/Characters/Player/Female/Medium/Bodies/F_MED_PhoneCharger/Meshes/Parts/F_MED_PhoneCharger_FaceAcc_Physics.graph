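# Physics asset for the PhoneCharger female medium character's FaceAcc mesh, defining physics for left and right dynamic hair chains.
# Each chain links four capsule bodies to a small sphere on its MSTR hair root bone.
# Constraint swing limits widen from root to tip, while twist stays uniformly tight.
# Unconstrained capsules on the head, neck and upper spine approximate the body, likely as colliders for the hair.
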PhysicsAsset /BRCosmetics/Characters/Player/Female/Medium/Bodies/F_MED_PhoneCharger/Meshes/Parts/F_MED_PhoneCharger_FaceAcc_Physics
bodies {
 MSTR_hair_l_1 [{shapeType=Sphere, radius=0.505, shapeName=MSTR_hair_left_roll_01_sphere}]
 MSTR_hair_r_1 [{shapeType=Sphere, radius=0.505, shapeName=MSTR_hair_left_roll_01_sphere}]
 dyn_hair_l_1 [{
   shapeType=Capsule
   length=3.2805
   radius=1.7337
   shapeName=dyn_hair_l_1_capsule
   center=<2.34 0.0459 1.0223>
   rotation=<rot -0.3869 89.7075 -92.8934>
  }]
 dyn_hair_l_2 [{
   shapeType=Capsule
   length=1
   radius=2.8058
   shapeName=dyn_hair_l_2_capsule
   center=<1.4441 0.2416 0.7448>
   rotation=<rot 1.3982 -88.9621 -90.4172>
  }]
 dyn_hair_l_3 [{shapeType=Capsule, length=0, radius=4, shapeName=dyn_hair_l_3_capsule, center=<1.6787 0.3398 0.7804>, rotation=<rot 2.0694 -89.922 -90.01>}]
 dyn_hair_l_4 [{
   shapeType=Capsule
   length=2.6017
   radius=3
   shapeName=dyn_hair_l_4_capsule
   center=<2.8385 0.2439 0.6883>
   rotation=<rot 1.1697 -94.1176 -90.9513>
  }]
 dyn_hair_r_1 [{
   shapeType=Capsule
   length=3.7213
   radius=1.293
   shapeName=dyn_hair_l_1_capsule
   center=<2.4876 -0.0675 -0.5759>
   rotation=<rot -0.4129 91.6888 90.7447>
  }]
 dyn_hair_r_2 [{
   shapeType=Capsule
   length=0.2986
   radius=3.2268
   shapeName=dyn_hair_l_2_capsule
   center=<1.4136 0.1561 -0.3057>
   rotation=<rot 1.3931 -89.619 89.5662>
  }]
 dyn_hair_r_3 [{
   shapeType=Capsule
   length=0
   radius=3.3151
   shapeName=dyn_hair_l_3_capsule
   center=<1.5629 0.1339 -0.1418>
   rotation=<rot 2.066 -86.3774 90.1179>
  }]
 dyn_hair_r_4 [{
   shapeType=Capsule
   length=2.6017
   radius=2.9192
   shapeName=dyn_hair_l_4_capsule
   center=<2.8973 0.4195 0.059>
   rotation=<rot 1.2372 -89.9023 89.1385>
  }]
 head [{shapeType=Capsule, length=6.1942, radius=6.0538, center=<1.6305 3.8295 0.0098>, rotation=<rot -0.0474 -68.4161 -89.9641>}]
 neck_01 [{
   shapeType=Capsule
   length=11.0087
   radius=4.6572
   shapeName=neck_01_capsule
   center=<0.1587 0.7496 -0.0192>
   rotation=<rot 1.4468 -99.4449 89.7594>
  }]
 spine_03 [{shapeType=Capsule, length=15, radius=15, shapeName=spine_03_capsule, center=<11.6742 -6.3294 0>}]
 spine_04 [{shapeType=Capsule, length=24.5813, radius=8.3295, shapeName=spine_04_capsule, center=<12.8405 -4.1998 -0.7118>}]
}
constraints [
 {bone1Name=dyn_hair_l_2, bone2Name=dyn_hair_l_1, swing1LimitDegrees=20, swing2LimitDegrees=20, twistLimitDegrees=5}
 {bone1Name=dyn_hair_l_3, bone2Name=dyn_hair_l_2, swing1LimitDegrees=20, swing2LimitDegrees=20, twistLimitDegrees=5}
 {bone1Name=dyn_hair_l_4, bone2Name=dyn_hair_l_3, swing1LimitDegrees=30, swing2LimitDegrees=30, twistLimitDegrees=5}
 {bone1Name=dyn_hair_r_2, bone2Name=dyn_hair_r_1, swing1LimitDegrees=20, swing2LimitDegrees=20, twistLimitDegrees=5}
 {bone1Name=dyn_hair_r_3, bone2Name=dyn_hair_r_2, swing1LimitDegrees=20, swing2LimitDegrees=20, twistLimitDegrees=5}
 {bone1Name=dyn_hair_r_4, bone2Name=dyn_hair_r_3, swing1LimitDegrees=30, swing2LimitDegrees=30, twistLimitDegrees=5}
 {bone1Name=dyn_hair_l_1, bone2Name=MSTR_hair_l_1, swing1LimitDegrees=10, swing2LimitDegrees=10, twistLimitDegrees=5}
 {bone1Name=dyn_hair_r_1, bone2Name=MSTR_hair_r_1, swing1LimitDegrees=10, swing2LimitDegrees=10, twistLimitDegrees=5}
]
